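# Unlit additive master material for particle flares, usable on Niagara sprites, ribbons and meshes.
# A tinted, exponent-adjusted mask texture times particle color feeds an emissive adjustment function; a camera offset function is also present.
# Opacity multiplies color alpha by a shading path switch: default blends depth fade with optional view-angle fade, deferred uses depth fade only, forward and mobile use raw particle alpha.
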Material /Game/Effects/Niagara/MasterMaterials/Flares/M_Basic_Particle_Flare
graph {
 properties {
  bUsedWithMeshDeformer=true
  bUsedWithMeshParticles=true
  bUsedWithNiagaraRibbons=true
  bUsedWithNiagaraSprites=true
  bUsedWithParticleSprites=true
  bUsedWithSkeletalMesh=true
  blendMode=BLEND_Additive
  pixelDepthOffsetMode=PDOM_Legacy
  shadingModel=MSM_Unlit
  usedShadingModels=Unlit
 }
 nodes [
  MaterialFunctionCall(Offset Amount=ScalarParameter() {group=Opacity, parameterName=Camera Offset}) {materialFunction=/Engine/Functions/Engine_MaterialFunctions02/WorldPositionOffset/CameraOffset}
  n1 = VectorParameter() {parameterName=Color, defaultValue=<rgba 1 1 1 1>}
  n2 = Power(
   Base=TextureSampleParameter2D() {parameterName=Mask, texture=/Game/Effects/Fort_Effects/Textures/Flares/T_GlowSar_6pt}.RGB
   Exp=ScalarParameter() {defaultValue=1, parameterName=Mask Exponent}
  ) {constExponent=1}
  n3 = ParticleColor()
  MaterialFunctionCall(Emissive=Multiply(A=@n1.RGB, B=Multiply(A=@n2, B=@n3.RGB))) {materialFunction=&MF_FX_EmissiveAdjustments}
  n4 = DepthFade(Opacity=@n3.A, FadeDistance=ScalarParameter() {defaultValue=50, group=Opacity, parameterName=DepthFade})
  n5 = Multiply(
   A=@n4
   B=MaterialFunctionCall("FadeSensitivity (optional)"=ScalarParameter() {group=Opacity, parameterName=Angle Fade Bias}) {materialFunction=&FadeBasedOnViewAngle}.Result
  )
  n6 = ScalarParameter() {group=Opacity, parameterName=Use Angle Fade}
  Multiply(A=@n1.A, B=ShadingPathSwitch(Default=LinearInterpolate(A=@n4, B=@n5, Alpha=@n6), Deferred=@n4, Forward=@n3.A, Mobile=@n3.A))
 ]
}
