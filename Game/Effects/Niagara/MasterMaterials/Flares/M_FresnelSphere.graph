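# Additive, unlit, two-sided master material for Niagara mesh particles that renders a glowing Fresnel sphere.
# Particle color scales a Fresnel term with a dynamic-parameter exponent, feeding emissive through MF_FX_EmissiveAdjustments and alpha through a depth fade.
# The faded alpha times the Ref dynamic parameter drives a subtle index-of-refraction value.
# JustSphere blends the Fresnel to solid and that refraction value to neutral.
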Material /Game/Effects/Niagara/MasterMaterials/Flares/M_FresnelSphere
graph {
 properties {
  bUseTranslucencyVertexFog=false
  bUsedWithNiagaraMeshParticles=true
  blendMode=BLEND_Additive
  pixelDepthOffsetMode=PDOM_Legacy
  refractionMethod=RM_IndexOfRefraction
  shadingModel=MSM_Unlit
  twoSided=true
  usedShadingModels=Unlit
 }
 nodes [
  n1 = ParticleColor()
  n2 = Fresnel(ExponentIn=DynamicParameter() {paramNames [Fresnel,Ref,Param3,Param4]}.Fresnel)
  n3 = Constant() {r=1}
  n4 = ScalarParameter() {parameterName=JustSphere}
  n5 = DepthFade(Opacity=Multiply(A=@n1.A, B=LinearInterpolate(A=@n2, B=@n3, Alpha=@n4)))
  n6 = DynamicParameter() {paramNames [Fresnel,Ref,Param3,Param4]}
  LinearInterpolate(A=LinearInterpolate(Alpha=Multiply(A=@n5, B=@n6.Ref)) {constA=1, constB=1.05}, B=Constant() {r=1}, Alpha=@n4)
  MaterialFunctionCall(Emissive=Multiply(A=@n1.RGB, B=LinearInterpolate(A=@n2, B=@n3, Alpha=@n4))) {materialFunction=&MF_FX_EmissiveAdjustments}
 ]
}
