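# Unlit, two-sided Niagara sprite material for glow flares, using alpha-composite blending and a camera offset.
# A squared linear falloff from blended UV0/UV1 forms the Mask reroute, alongside a selectable three-shape Lerp_3Color branch.
# An inverse-decay curve, shaped by dynamic parameter IPD_K and capped by MaxiumBright, drives brightness.
# Particle color and alpha scale that curve for emissive and opacity, with reroutes providing depth and camera fades.
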Material /Game/Effects/Niagara/MasterMaterials/Flares/M_InverseDecay_Flare
graph {
 properties {
  bAutomaticallySetUsageInEditor=false
  bUsedWithNiagaraSprites=true
  blendMode=BLEND_AlphaComposite
  shadingModel=MSM_Unlit
  twoSided=true
  usedShadingModels=Unlit
 }
 nodes [
  MaterialFunctionCall(Offset Amount=ScalarParameter() {defaultValue=10, parameterName=CamOffset}) {materialFunction=/Engine/Functions/Engine_MaterialFunctions02/WorldPositionOffset/CameraOffset}
  n1 = TextureCoordinate()
  n2 = TextureCoordinate() {coordinateIndex=1}
  n3 = ScalarParameter() {parameterName=CubeMap_Switch}
  n4 = Subtract(A=Multiply(A=LinearInterpolate(A=@n1, B=@n2, Alpha=@n3)) {constB=2})
  n5 = MaterialFunctionCall(Float2=OneMinus(None=Abs(None=@n4))) {materialFunction=&BreakOutFloat2Components}
  n6 = Multiply(A=@n5.R, B=@n5.R)
  n7 = TextureCoordinate()
  n8 = ComponentMask(None=@n7) {r=true}
  n9 = ComponentMask(None=@n7) {g=true}
  n10 = Abs(None=Subtract(A=OneMinus(None=@n8), B=@n9))
  n11 = Abs(None=Subtract(A=@n8, B=@n9))
  n12 = Constant() {r=0.5}
  MaterialFunctionCall(
   A=@n6
   B=OneMinus(None=Add(A=@n10, B=@n11))
   C=SphereMask(A=TextureCoordinate(), B=@n12, Radius=@n12) {hardnessPercent=0}
   Alpha=ScalarParameter() {defaultValue=0.5, parameterName=Mask_Select}
  ) {materialFunction=&Lerp_3Color}
  n13 = ParticleColor()
  n14 = OneMinus(None=NamedRerouteUsage())
  n15 = DynamicParameter() {defaultValue=<rgba 1 1 1 0.01>, paramNames [Emi,Param2,Param3,IPD_K]}
  n16 = Min(A=Subtract(A=Divide(A=@n15.IPD_K, B=@n14), B=@n15.IPD_K), B=ScalarParameter() {defaultValue=1000, parameterName=MaxiumBright}) {constB=1000}
  n17 = Multiply(A=NamedRerouteUsage(), B=NamedRerouteUsage())
  MaterialFunctionCall(Emissive=Multiply(A=Multiply(A=@n13.RGB, B=@n16), B=@n17)) {materialFunction=&MF_FX_EmissiveAdjustments}
  Multiply(A=Multiply(A=@n13.A, B=@n16), B=@n17)
  NamedRerouteDeclaration(None=DepthFade(FadeDistance=ScalarParameter() {defaultValue=20, parameterName=DF_FD})) {name=DepthFade, nodeColor=<rgba 1 1 1 1>}
  n18 = Subtract(A=CameraPositionWS(), B=WorldPosition().XYZ)
  n19 = Length(None=@n18) {is=class MaterialExpressionLength}
  n20 = DotProduct(A=Normalize(VectorInput=@n18), B=Transform(None=Constant3Vector() {constant=<rgba 0 0 -1 0>}) {transformSourceType=TRANSFORMSOURCE_View})
  n21 = ScalarParameter() {defaultValue=20, parameterName=Camfade_Offset}
  n22 = ScalarParameter() {defaultValue=20, parameterName=Camfade_Lemgth}
  n23 = Divide(A=Subtract(A=Multiply(A=@n19, B=@n20), B=@n21), B=@n22)
  n24 = Divide(A=Subtract(A=PixelDepth().R, B=@n21), B=@n22)
  n25 = StaticBool()
  NamedRerouteDeclaration(None=Saturate(None=StaticSwitch(True=@n23, False=@n24, Value=@n25))) {name=CamFade, nodeColor=<rgba 1 0 0.9613 1>}
  NamedRerouteDeclaration(None=@n6) {name=Mask, nodeColor=<rgba 1 1 1 1>}
 ]
}
