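# Translucent unlit master material for Niagara sprite and ribbon effects, defaulting to a smoke texture.
# A main texture blends with a panned noise multiply, timed by game time or a dynamic parameter, then takes a per-particle erode bias.
# Selected channels drive opacity and either a two-color tint or a curve-atlas gradient.
# Optional depth and near-camera fades shape opacity, emissive adjustment is blendable, and particle color multiplies the result.
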Material /Game/Effects/Niagara/MasterMaterials/Flares/M_NFX_Sprite
graph {
 properties {
  bUseMaterialAttributes=true
  bUsedWithNiagaraRibbons=true
  bUsedWithNiagaraSprites=true
  blendMode=BLEND_Translucent
  pixelDepthOffsetMode=PDOM_Legacy
  shadingModel=MSM_Unlit
  usedShadingModels=Unlit
 }
 nodes [
  n1 = VectorParameter() {group=P3_FinalTint, parameterName=BaseColorMutiply, defaultValue=<rgba 1 1 1 0>}
  n2 = VectorParameter() {group=P2_AlphaAndColorTint, parameterName=ColorTintA_DarkPart, defaultValue=<rgba 0.5104 0.0469 0.0316 0>}
  n3 = TextureSampleParameter2D() {group=P1_1_Textures, parameterName=MainTexture, texture=/Game/Effects/Fort_Effects/Textures/Husk/T_Husk_Smoke}
  n4 = LinearInterpolate(
   A=TextureCoordinate()
   B=ParticleMacroUV()
   Alpha=ScalarParameter() {group=P1_2_TilingTextures, parameterName="0_TillingTexture_UseMacroUV_(0or1)"}
  )
  n5 = AppendVector(
   A=ScalarParameter() {defaultValue=1, group=P1_2_TilingTextures, parameterName=TilingTexture_U_Repeat}
   B=ScalarParameter() {defaultValue=1, group=P1_2_TilingTextures, parameterName=TilingTexture_V_Repeat}
  )
  n6 = ScalarParameter() {defaultValue=1, group=P1_2_TilingTextures, parameterName=TilingTexture_U_Speed}
  n7 = ScalarParameter() {defaultValue=1, group=P1_2_TilingTextures, parameterName=TilingTexture_V_Speed}
  n8 = Time()
  n9 = DynamicParameter() {defaultValue=<rgba 0 0 0 0>, paramNames [MaterialTime, Param2, Param3, FullChannelErodeBais]}
  n10 = ScalarParameter() {group=0_GeneralSettings, parameterName="UseDPControlTime(0or1)"}
  n11 = TextureSampleParameter2D(UVs=Panner(Coordinate=Multiply(A=@n4, B=@n5), Time=LinearInterpolate(A=@n8, B=@n9.MaterialTime, Alpha=@n10), Speed=AppendVector(A=@n6, B=@n7))) {
   group=P1_2_TilingTextures
   parameterName=TilingTexture
   samplerType=SAMPLERTYPE_LinearColor
   texture=&T_NoisePack_LinearColor
  }
  n12 = Max(A=LinearInterpolate(
    A=@n3.RGBA
    B=Multiply(A=@n3.RGBA, B=@n11.RGBA)
    Alpha=ScalarParameter() {defaultValue=0.25, group=P1_2_TilingTextures, parameterName=TilingTexture_Multiply}
   ) {constAlpha=0.2}) {constB=0}
  n13 = Multiply(
   A=Saturate(None=Ceil(None=@n12))
   B=Add(
    A=@n12
    B=DynamicParameter() {defaultValue=<rgba 0 0 0 0>, paramNames [MaterialTime, Param2, Param3, FullChannelErodeBais]}.FullChannelErodeBais
   )
  )
  n14 = ChannelMaskParameter(None=@n13) {group=P2_AlphaAndColorTint, maskChannel=Green, parameterName=ColorTintChannel, defaultValue=<rgba 0 1 0 0>}
  n15 = CurveAtlasRowParameter(CurveTime=@n14.R) {
   atlas=&AthenaCurveAtlas01
   curve=/Game/Athena/LinearColorCurves/Curve_RiftSpawn
   group=P2_AlphaAndColorTint
   parameterName=TintCurve
  }
  n16 = ScalarParameter() {group=P2_AlphaAndColorTint, parameterName="UseCurveTintColorAndAlpha(0or1)"}
  n17 = LinearInterpolate(
   A=LinearInterpolate(
    A=@n2.RGB
    B=VectorParameter() {group=P2_AlphaAndColorTint, parameterName=ColorTintB_BrightPart, defaultValue=<rgba 3 0.7511 0.1494 0>}.RGB
    Alpha=@n14.R
   )
   B=@n15
   Alpha=@n16
  )
  n18 = Multiply(A=@n17, B=VectorParameter() {group=P3_FinalTint, parameterName=EmissiveColorMutiply, defaultValue=<rgba 1 1 1 0>}.RGB)
  n19 = MaterialFunctionCall(Emissive=@n18) {materialFunction=&MF_FX_EmissiveAdjustments}
  n20 = ScalarParameter() {parameterName=USE_EMISSIVE_LERP}
  n21 = ChannelMaskParameter(None=@n13) {group=P2_AlphaAndColorTint, maskChannel=Green, parameterName=AlphaChannel, defaultValue=<rgba 0 1 0 0>}
  n22 = LinearInterpolate(A=@n21.R, B=@n15.A, Alpha=@n16)
  n23 = ScalarParameter() {defaultValue=500, group=Fade Controls, parameterName=DepthFade}
  n24 = StaticSwitchParameter(True=Multiply(A=@n22, B=DepthFade(FadeDistance=@n23)), False=@n22) {group=Fade Controls, parameterName=UseDepthFade}
  n25 = Multiply(
   A=@n24
   B=MaterialFunctionCall(Distance=ScalarParameter() {defaultValue=1500, group=Fade Controls, parameterName=NearCameraFadeDistance}) {materialFunction=&Near_Camera_Fade}.Result
  )
  n26 = GetMaterialAttributes(None=SetMaterialAttributes!(
    MaterialAttributes
    Base Color=Multiply(A=@n17, B=@n1.RGB)
    Emissive Color=LinearInterpolate(A=@n18, B=@n19.Result, Alpha=@n20)
    Opacity=StaticSwitchParameter(True=@n25, False=@n24) {group=Fade Controls, parameterName=UseCameraNearFade}
   ) {attributeSetTypes [#1, #2, #3]}) {attributeGetTypes [#1, #2, #3]}
  n27 = ParticleColor()
  SetMaterialAttributes(
   MaterialAttributes=@n26.MaterialAttributes
   Base Color=Multiply(A=@n27.RGB, B=@n26.Base Color)
   Emissive Color=Multiply(A=@n27.RGB, B=@n26.Emissive Color)
   Opacity=Multiply(A=@n26.Opacity, B=@n27.A)
  ) {attributeSetTypes [#1, #2, #3]}
  StaticSwitchParameter() {group=P2_AlphaAndColorTint, parameterName=UsingCurveTintColorAndAlpha}
  StaticSwitchParameter() {group=P2_AlphaAndColorTint, parameterName=UsingCurveTintColorAndAlpha}
 ]
}
