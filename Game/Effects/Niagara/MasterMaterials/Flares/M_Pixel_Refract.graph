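# Translucent, unlit, two-sided Niagara sprite material for a pixel-flare effect with chromatic fringing and index-of-refraction distortion.
# Dynamic ChromaOffset shifts two tinted flare samples in opposite directions, which are blended, saturation-adjusted and added to the particle-colored flare before emissive adjustments.
# The flare's green channel, particle alpha, depth fade and near-camera fade mask the dynamic Refraction strength.
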Material /Game/Effects/Niagara/MasterMaterials/Flares/M_Pixel_Refract
graph {
 properties {
  bUsedWithNiagaraSprites=true
  blendMode=BLEND_Translucent
  pixelDepthOffsetMode=PDOM_Legacy
  refractionMethod=RM_IndexOfRefraction
  shadingModel=MSM_Unlit
  twoSided=true
  usedShadingModels=Unlit
 }
 nodes [
  n1 = DynamicParameter() {paramNames [ChromaOffset,Refraction,Random,Param4]}
  AppendVector(
   A=MaterialFunctionCall!(FlickerInitialSpeed=@n1.Random, FlickerRandom) {materialFunction=/Game/Effects/Fort_Effects/Materials/_Functions/MF_Flicker}.Result
   B=MaterialFunctionCall(FlickerRandom=Add(A=@n1.Random) {constB=0.32}) {materialFunction=&MF_Flicker}.Result
  )
  n2 = ParticleColor()
  n3 = Multiply(
   A=DepthFade() {fadeDistanceDefault=50}
   B=MaterialFunctionCall(Distance=ScalarParameter() {defaultValue=150, parameterName=NearCamFade}) {materialFunction=&Near_Camera_Fade}.Result
  )
  n4 = TextureSample() {texture=/Game/Effects/Fort_Effects/Textures/Flares/T_Flare_Pixel}
  LinearInterpolate(B=@n1.Refraction, Alpha=Multiply(A=@n2.A, B=Multiply(A=@n3, B=@n4.G))) {constA=1, constB=1.1}
  LinearInterpolate(A=Constant3Vector() {constant=<rgba 1 0.5 1 0>}, B=Constant3Vector() {constant=<rgba 0.5 1 1 0>}, Alpha=@n1.ChromaOffset)
  n5 = TextureCoordinate()
  n6 = Multiply(A=Constant2Vector() {g=0.05, r=0.05}, B=@n1.ChromaOffset)
  n7 = Add(A=@n5, B=@n6)
  n8 = Subtract(A=@n5, B=@n6)
  n9 = LinearInterpolate(
   A=Multiply(
    A=VectorParameter() {parameterName=Chroma1, defaultValue=<rgba 1.5 0 0.8249 1>}.RGB
    B=TextureSample(UVs=@n7) {texture=&T_Flare_Pixel}.R
   )
   B=Multiply(
    A=TextureSample(UVs=@n8) {texture=&T_Flare_Pixel}.R
    B=VectorParameter() {parameterName=Chroma2, defaultValue=<rgba 0 1.5 1.2326 0>}.RGB
   )
  )
  n10 = ScalarParameter() {defaultValue=-2, parameterName=Desaturation}
  MaterialFunctionCall(Emissive=Add(A=Multiply(A=@n2.RGB, B=@n4.R), B=Desaturation(None=@n9, Fraction=@n10))) {materialFunction=&MF_FX_EmissiveAdjustments}
 ]
}
