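# Translucent unlit master material for spark flare sprites in Niagara and particle sprite effects.
# Particle color RGB feeds MF_FX_EmissiveAdjustments for the glow.
# Opacity multiplies particle alpha by the red channel of a texture parameter, which defaults to an AntiMaterialCharge sparks texture.
# MF_LowMobile_PathingSwitch applies a parameterized depth fade on the default path and skips it on low-end and mobile.
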Material /Game/Effects/Niagara/MasterMaterials/Flares/M_SparkFlare
graph {
 properties {
  bUsedWithNiagaraSprites=true
  bUsedWithParticleSprites=true
  blendMode=BLEND_Translucent
  pixelDepthOffsetMode=PDOM_Legacy
  shadingModel=MSM_Unlit
  usedShadingModels=Unlit
 }
 nodes [
  n1 = ParticleColor()
  MaterialFunctionCall(Emissive=@n1.RGB) {materialFunction=&MF_FX_EmissiveAdjustments}
  n2 = Multiply(
   A=@n1.A
   B=TextureSampleParameter2D(UVs=TextureCoordinate()) {parameterName=Texture, texture=/Game/CharClasses/Outlander/Gadgets/AntiMaterialCharge/Materials/T_AntiMaterialCharge_Sparks_01}.R
  )
  n3 = ScalarParameter() {parameterName=Depthfade}
  MaterialFunctionCall!(Default=Multiply(A=@n2, B=DepthFade(FadeDistance=@n3)), Low/Mobile=@n2) {materialFunction=/Game/Effects/Fort_Effects/Materials/_Functions/MF_LowMobile_PathingSwitch}
 ]
}
