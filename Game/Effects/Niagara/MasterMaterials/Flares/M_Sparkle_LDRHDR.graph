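# Translucent unlit Niagara and sprite particle material for sparkle flares, with separate LDR and HDR color handling.
# Particle color tints a per-particle random blend of Color A and B; mobile uses only Color A.
# The default shading path derives HDR from the clamped color, while Forward and Mobile pass it straight to emissive adjustments.
# Opacity combines particle alpha, texture red, exponent-shaped green and a pixel-depth near-camera fade.
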Material /Game/Effects/Niagara/MasterMaterials/Flares/M_Sparkle_LDRHDR
graph {
 properties {
  bUsedWithNiagaraMeshParticles=true
  bUsedWithNiagaraSprites=true
  bUsedWithParticleSprites=true
  blendMode=BLEND_Translucent
  pixelDepthOffsetMode=PDOM_Legacy
  shadingModel=MSM_Unlit
  usedShadingModels=Unlit
 }
 nodes [
  Constant()
  n1 = ParticleColor()
  n2 = VectorParameter() {parameterName=Color A, defaultValue=<rgba 1 0.5534 0.295 0>}
  n3 = VectorParameter() {parameterName=Color B, defaultValue=<rgba 0.401 0.3782 1 0>}
  n4 = ParticleRandom()
  n5 = Multiply(A=@n1.RGB, B=ShadingPathSwitch(Default=LinearInterpolate(A=@n2.RGB, B=@n3.RGB, Alpha=@n4), Mobile=@n2.RGB))
  n6 = Clamp(None=@n5)
  MaterialFunctionCall(Emissive=ShadingPathSwitch(
    Default=MaterialFunctionCall(LDR Input=@n6, HDR Tint=@n5) {materialFunction=&DeriveHDRfromLDR}.RemappedRange
    Forward=@n5
    Mobile=@n5
   )) {materialFunction=&MF_FX_EmissiveAdjustments}
  n7 = TextureSampleParameter2D() {parameterName=SparkleTexture, texture=/Game/Effects/Fort_Effects/Textures/Sparkle/T_Sparkle_01}
  n8 = Power(Base=@n7.G, Exp=ScalarParameter() {defaultValue=1, parameterName=Exponent})
  n9 = PixelDepth()
  n10 = Constant()
  n11 = ScalarParameter() {parameterName=CameraDistanceFade}
  Multiply(A=Multiply(A=@n1.A, B=Multiply(A=@n7.R, B=@n8)), B=OneMinus(None=SphereMask(A=@n9.R, B=@n10, Radius=@n11) {hardnessPercent=10}))
 ]
}
