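# Unlit additive Niagara sprite material for a sharp blade-sheen sparkle flare, tinted by particle color.
# A static switch picks a sparkle texture or a procedural star built from centered UVs plus a 45-degree rotated copy, computed through a vertex interpolator.
# Curvature and exponent powers shape the crossed points before MF_FX_EmissiveAdjustments.
# Particle alpha has an optional depth fade, bypassed on mobile.
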Material /Game/Effects/Niagara/MasterMaterials/Flares/M_Sparkle_SharpBladeSheen
graph {
 properties {
  bUsedWithNiagaraSprites=true
  blendMode=BLEND_Additive
  pixelDepthOffsetMode=PDOM_Legacy
  shadingModel=MSM_Unlit
  translucencyPass=MTP_BeforeDOF
  usedShadingModels=Unlit
 }
 nodes [
  n1 = ParticleColor()
  n2 = Multiply(A=@n1.A, B=DepthFade(FadeDistance=ScalarParameter() {defaultValue=10, parameterName=DepthFadeDistance}))
  LinearInterpolate(
   A=@n1.A
   B=ShadingPathSwitch(Default=@n2, Deferred=@n2, Mobile=@n1.A) {desc=Migrated from FeatureLevelSwitch}
   Alpha=ScalarParameter() {parameterName=UseDepthFade}
  )
  n3 = Divide(A=Subtract(A=TextureCoordinate()) {constB=0.5})
  n4 = ComponentMask(None=@n3) {r=true}
  n5 = Constant() {r=0.125}
  n6 = Cosine(None=@n5)
  n7 = Multiply(A=@n4, B=@n6)
  n8 = ComponentMask(None=@n3) {g=true}
  n9 = Sine(None=@n5)
  n10 = Multiply(A=@n8, B=@n9)
  n11 = Multiply(A=@n6, B=@n8)
  n12 = Multiply(A=@n9, B=@n4)
  n13 = AppendVector(A=@n3, B=AppendVector(A=Add(A=@n7, B=@n10), B=Subtract(A=@n11, B=@n12)))
  n14 = Power(
   Base=Abs(None=VertexInterpolator(VS=@n13).PS)
   Exp=ScalarParameter() {defaultValue=0.333, group=Sparkle, parameterName=Sparkle Curvature, sliderMax=0.5, sliderMin=0.333}
  )
  n15 = ComponentMask(None=@n14) {g=true, r=true}
  n16 = OneMinus(None=Saturate(None=MaterialFunctionCall(f2=@n15) {materialFunction=&AddComponents}.f2))
  n17 = ComponentMask(None=@n14) {a=true, b=true}
  n18 = Multiply(
   A=OneMinus(None=Divide(A=@n16) {constB=3})
   B=Saturate(None=MaterialFunctionCall(f2=@n17) {materialFunction=&AddComponents}.f2)
  )
  n19 = Power(
   Base=OneMinus(None=Saturate(None=@n18))
   Exp=ScalarParameter() {defaultValue=4, group=Sparkle, parameterName=Sparkle Exponent, sliderMax=9, sliderMin=4}
  )
  n20 = TextureSampleParameter2D() {group=Sparkle, parameterName=Sparkle Texture, texture=/Game/Effects/Fort_Effects/Textures/Sparkle/T_Sparkle_SharpBlade}
  MaterialFunctionCall(Emissive=Multiply(A=StaticSwitchParameter(True=@n19, False=@n20.RGB) {group=Sparkle, parameterName=Use Procedural Sparkle}, B=@n1.RGB)) {materialFunction=&MF_FX_EmissiveAdjustments}
 ]
}
