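# Translucent unlit particle material for big twinkling four-point star sparkles, named for the Unicorn Pickaxe effect.
# Particle random drives a sine twinkle speed and a slight hue shift of the brightened particle color, blended against plain particle color before emissive adjustments.
# Opacity multiplies particle alpha by the star texture's green channel, with depth fade skipped on low/mobile.
# A camera offset function supplies a parameterized world position offset.
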Material /Game/Effects/Niagara/MasterMaterials/Flares/M_Unicorn_Pickaxe_Sparkles_Big
graph {
 properties {
  bEnableResponsiveAA=true
  bUsedWithNanite=true
  bUsedWithNiagaraSprites=true
  bUsedWithParticleSprites=true
  blendMode=BLEND_Translucent
  pixelDepthOffsetMode=PDOM_Legacy
  shadingModel=MSM_Unlit
  usedShadingModels=Unlit
 }
 nodes [
  n1 = ParticleColor()
  n2 = ParticleRandom()
  n3 = Multiply(A=@n2) {constB=0.25}
  n4 = MaterialFunctionCall(Speed=Multiply(A=@n2, B=ScalarParameter() {defaultValue=3, parameterName=TwinkleSpeed})) {materialFunction=&DebugTimeSine}
  n5 = Multiply(A=Clamp(None=Multiply(A=@n4.Result) {constB=5}) {maxDefault=5, minDefault=0.25}, B=@n1.RGB)
  MaterialFunctionCall(Emissive=LinearInterpolate(
    A=@n1.RGB
    B=MaterialFunctionCall(Hue Shift Percentage=@n3, Texture=@n5) {materialFunction=&HueShift}.Result
    Alpha=ScalarParameter() {defaultValue=1, parameterName=Use Particle Random Color}
   )) {materialFunction=&MF_FX_EmissiveAdjustments}
  MaterialFunctionCall(Offset Amount=ScalarParameter() {defaultValue=25, parameterName="Offset Amount (S)"}) {materialFunction=/Engine/Functions/Engine_MaterialFunctions02/WorldPositionOffset/CameraOffset}
  n6 = Multiply(
   A=ParticleColor().A
   B=TextureSampleParameter2D() {parameterName=Texture, texture=/Game/Effects/Fort_Effects/Textures/Sparks/T_Sparkles_4Star_01_D}.G
  )
  n7 = ScalarParameter() {parameterName=Depthfade}
  MaterialFunctionCall!(Default=Multiply(A=@n6, B=DepthFade(FadeDistance=@n7)), Low/Mobile=@n6) {materialFunction=/Game/Effects/Fort_Effects/Materials/_Functions/MF_LowMobile_PathingSwitch}
 ]
}
